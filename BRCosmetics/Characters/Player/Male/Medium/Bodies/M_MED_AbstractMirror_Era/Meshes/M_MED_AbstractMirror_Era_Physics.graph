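# Physics asset for the AbstractMirror_Era medium male player body cosmetic, defining secondary-motion physics for left, center and right chest loops.
# Each loop pairs a small master capsule with a larger dyn capsule, joined by a tight constraint that allows modest, uneven swing and almost no twist.
# All three master bodies attach to a large spine_05 torso capsule through constraints with free swing and twist, leaving their listed angle limits inactive.
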PhysicsAsset /BRCosmetics/Characters/Player/Male/Medium/Bodies/M_MED_AbstractMirror_Era/Meshes/M_MED_AbstractMirror_Era_Physics
bodies {
 dyn_C_chest_loop_jnt [{shapeType=Capsule, length=2, radius=2, shapeName=dyn_C_chest_loop_jnt_capsule, center=<2 0 0>, rotation=<rot -0.4796 -90 -0.3644>}]
 dyn_L_chest_loop_jnt [{shapeType=Capsule, length=2, radius=2, shapeName=dyn_L_chest_loop_jnt_capsule, center=<2 0 0>, rotation=<rot -2.2042 -90 0.2012>}]
 dyn_R_chest_loop_jnt [{shapeType=Capsule, length=2, radius=2, shapeName=dyn_R_chest_loop_jnt_capsule, center=<2 0 0>, rotation=<rot -3.974 -90 2.7777>}]
 master_C_chest_loop_jnt [{shapeType=Capsule, length=0.505, radius=0.505, shapeName=master_C_chest_loop_jnt_capsule, rotation=<rot 0 -90 0>}]
 master_L_chest_loop_jnt [{shapeType=Capsule, length=0.505, radius=0.505, shapeName=master_L_chest_loop_jnt_capsule, rotation=<rot 0 -90 0>}]
 master_R_chest_loop_jnt [{shapeType=Capsule, length=0.505, radius=0.505, shapeName=master_R_chest_loop_jnt_capsule, rotation=<rot 0 -90 0>}]
 spine_05 [{shapeType=Capsule, length=12, radius=10, shapeName=spine_05_capsule}]
}
constraints [
 {bone1Name=dyn_C_chest_loop_jnt, bone2Name=master_C_chest_loop_jnt, swing1LimitDegrees=25, swing2LimitDegrees=5, twistLimitDegrees=1}
 {bone1Name=dyn_L_chest_loop_jnt, bone2Name=master_L_chest_loop_jnt, swing1LimitDegrees=25, swing2LimitDegrees=5, twistLimitDegrees=1}
 {bone1Name=dyn_R_chest_loop_jnt, bone2Name=master_R_chest_loop_jnt, swing1LimitDegrees=25, swing2LimitDegrees=5, twistLimitDegrees=1}
 {
  bone1Name=master_L_chest_loop_jnt
  bone2Name=spine_05
  swing1LimitDegrees=45
  swing1Motion=Free
  swing2LimitDegrees=45
  swing2Motion=Free
  twistLimitDegrees=45
  twistMotion=Free
 }
 {
  bone1Name=master_C_chest_loop_jnt
  bone2Name=spine_05
  swing1LimitDegrees=45
  swing1Motion=Free
  swing2LimitDegrees=45
  swing2Motion=Free
  twistLimitDegrees=45
  twistMotion=Free
 }
 {
  bone1Name=master_R_chest_loop_jnt
  bone2Name=spine_05
  swing1LimitDegrees=45
  swing1Motion=Free
  swing2LimitDegrees=45
  swing2Motion=Free
  twistLimitDegrees=45
  twistMotion=Free
 }
]
