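# Post-process style AnimBP for the AbstractMirror_Era medium male body cosmetic, copying the parent mesh pose and adding secondary motion to its gear.
# A Control Rig and weighted multi-bone constraints place the pouch, loop and holster master joints, then cache StartPose.
# Holster AnimDynamics runs at reduced alpha when bIsInFrontEnd is true, full otherwise.
# Chest pouch AnimDynamics are zeroed (alpha 0), followed by a RigidBody pass using the body's physics asset.
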AnimBlueprint /BRCosmetics/Characters/Player/Male/Medium/Bodies/M_MED_AbstractMirror_Era/Meshes/M_MED_AbstractMirror_Era_AnimBP
target_skeleton=&/M_MED_AbstractMirror_Era_Skeleton
graphs [
 AnimGraph {nodes [
   n1 = Anim Dynamics - Bone: dyn_gun_holster_jnt(ComponentPose=Local To Component(LocalPose=Use cached pose 'StartPose'()), Alpha=1, ExternalForce="0,0,0") {
    angularDampingOverride=0.99
    angularSpringConstant=20000
    bAngularSpring=true
    bOverrideAngularDamping=true
    bUseGravityOverride=true
    simulationSpace=BoneRelative
    alphaBoolBlend {blendOption=Linear}
    boundBone {boneName=dyn_gun_holster_jnt}
    componentAppliedLinearAccClamp=<3000 3000 3000>
    componentLinearAccScale=<0.3 0.3 0.5>
    componentLinearVelScale=<0.3 0.3 0.5>
    gravityOverride {z=-980}
    physicsBodyDefinitions [{
      boundBone {boneName=dyn_gun_holster_jnt}
      boxExtents=<35 35 10>
      constraintSetup {angularLimitsMax=<10 10 35>, angularLimitsMin=<-10 5 -35>, angularTarget=<0 0 0>}
      localJointOffset=<0 0 -2>
     }]
    relativeSpaceBone {boneName=thigh_r}
   }
   n2 = Component To Local(ComponentPose=@n1)
   n3 = Anim Dynamics - Bone: dyn_gun_holster_jnt(ComponentPose=Local To Component(LocalPose=Use cached pose 'StartPose'()), Alpha=0.25, ExternalForce="0,0,0") {
    angularDampingOverride=0.99
    angularSpringConstant=20000
    bAngularSpring=true
    bOverrideAngularDamping=true
    bUseGravityOverride=true
    simulationSpace=BoneRelative
    alphaBoolBlend {blendOption=Linear}
    boundBone {boneName=dyn_gun_holster_jnt}
    componentAppliedLinearAccClamp=<3000 3000 3000>
    componentLinearAccScale=<0.3 0.3 0.5>
    componentLinearVelScale=<0.3 0.3 0.5>
    gravityOverride {z=-980}
    physicsBodyDefinitions [{
      boundBone {boneName=dyn_gun_holster_jnt}
      boxExtents=<35 35 10>
      constraintSetup {angularLimitsMax=<10 10 35>, angularLimitsMin=<-10 5 -35>, angularTarget=<0 0 0>}
      localJointOffset=<0 0 -2>
     }]
    relativeSpaceBone {boneName=thigh_r}
   }
   n4 = Component To Local(ComponentPose=@n3)
   n5 = Get bIsInFrontEnd()
   n6 = Anim Dynamics - Bone: dyn_L_chest_pouch_jnt(
    ComponentPose=Local To Component(LocalPose=Blend Poses by bool(BlendPose_0=@n4, BlendPose_1=@n2, bActiveValue=@n5) {blendType=Linear})
    Alpha=0
    ExternalForce="0,0,0"
   ) {
    angularDampingOverride=1
    angularSpringConstant=2000
    bAngularSpring=true
    bOverrideAngularDamping=true
    bUseGravityOverride=true
    simulationSpace=RootRelative
    alphaBoolBlend {blendOption=Linear}
    boundBone {boneName=dyn_L_chest_pouch_jnt}
    componentAppliedLinearAccClamp=<3000 3000 3000>
    componentLinearAccScale=<0.2 0.2 0.2>
    componentLinearVelScale=<0.2 0.2 0.2>
    gravityOverride {z=-500}
    physicsBodyDefinitions [{
      boundBone {boneName=dyn_L_chest_pouch_jnt}
      boxExtents=<20 10 10>
      constraintSetup {angularLimitsMax=<5 5 2>, angularLimitsMin=<-5 -5 -2>, angularTarget=<0 0 0>}
      localJointOffset=<5 0 0>
     }]
    relativeSpaceBone {boneName=spine_05}
   }
   n7 = Anim Dynamics - Bone: dyn_R_chest_pouch_jnt(ComponentPose=@n6, Alpha=0, ExternalForce="0,0,0") {
    angularDampingOverride=1
    angularSpringConstant=2000
    bAngularSpring=true
    bOverrideAngularDamping=true
    bUseGravityOverride=true
    simulationSpace=RootRelative
    alphaBoolBlend {blendOption=Linear}
    boundBone {boneName=dyn_R_chest_pouch_jnt}
    componentAppliedLinearAccClamp=<3000 3000 3000>
    componentLinearAccScale=<0.2 0.2 0.2>
    componentLinearVelScale=<0.2 0.2 0.2>
    gravityOverride {z=-500}
    physicsBodyDefinitions [{
      boundBone {boneName=dyn_R_chest_pouch_jnt}
      boxExtents=<20 10 10>
      constraintSetup {angularLimitsMax=<5 5 2>, angularLimitsMin=<-5 -5 -2>, angularTarget=<0 0 0>}
      localJointOffset=<5 0 0>
     }]
    relativeSpaceBone {boneName=spine_05}
   }
   n8 = Anim Dynamics - Bone: dyn_C_chest_pouch_jnt(ComponentPose=@n7, Alpha=0, ExternalForce="0,0,0") {
    angularDampingOverride=1
    angularSpringConstant=2000
    bAngularSpring=true
    bOverrideAngularDamping=true
    bUseGravityOverride=true
    simulationSpace=RootRelative
    alphaBoolBlend {blendOption=Linear}
    boundBone {boneName=dyn_C_chest_pouch_jnt}
    componentAppliedLinearAccClamp=<3000 3000 3000>
    componentLinearAccScale=<0.2 0.2 0.2>
    componentLinearVelScale=<0.2 0.2 0.2>
    gravityOverride {z=-500}
    physicsBodyDefinitions [{
      boundBone {boneName=dyn_C_chest_pouch_jnt}
      boxExtents=<20 10 10>
      constraintSetup {angularLimitsMax=<5 5 2>, angularLimitsMin=<-5 -5 -2>, angularTarget=<0 0 0>}
      localJointOffset=<5 0 0>
     }]
    relativeSpaceBone {boneName=spine_05}
   }
   n9 = RigidBody(ComponentPose=@n8) {
    bOverrideWorldGravity=true
    overridePhysicsAsset=&/M_MED_AbstractMirror_Era_Physics
    simulationSpace=BaseBoneSpace
    alphaBoolBlend {blendOption=Linear}
    baseBoneRef {boneName=spine_04}
    componentLinearAccScale=<0.3 0.2 0.2>
    componentLinearVelScale=<0.3 0.2 0.2>
    overrideWorldGravity {z=-980}
    simSpaceSettings {worldAlpha=0.5}
   }
   Output Pose(Result=Component To Local(ComponentPose=@n9)) {layerGroup=None}
   n10 = Control Rig(Source=Copy Pose From Mesh() {bUseAttachedParent=true}) {
    alphaBoolBlend {blendOption=Linear}
    controlRigAssetReference {blueprintRigClass=&/M_MED_AbstractMirror_Era_CtrlRig.M_MED_AbstractMirror_Era_CtrlRig_C}
   }
   n11 = "Constraint - master_R_chest_pouch_jnt (PPPPPPP)"!(
    ComponentPose=Local To Component(LocalPose=@n10)
    Alpha=1
    ConstraintWeights_0=0.031
    ConstraintWeights_1:float=0.0939
    ConstraintWeights_2:float=0.232
    ConstraintWeights_3:float=0.3885
    ConstraintWeights_4:float=0.1497
    ConstraintWeights_5:float=0.0651
    ConstraintWeights_6:float=0.0397
   ) -> (Pose) {
    alphaBoolBlend {blendOption=Linear}
    boneToModify {boneName=master_R_chest_pouch_jnt}
    constraintSetup [
     {targetBone {boneName=pelvis}}
     {targetBone {boneName=spine_01}}
     {targetBone {boneName=spine_02}}
     {targetBone {boneName=spine_03}}
     {targetBone {boneName=spine_04}}
     {targetBone {boneName=clavicle_r}}
     {targetBone {boneName=deform_pec_r}}
    ]
    constraintWeights [1,1,1,1,1,1,1]
   }
   n12 = "Constraint - master_C_chest_pouch_jnt (PPPPPPPP)"!(
    ComponentPose=@n11
    Alpha=1
    ConstraintWeights_0=0.0367
    ConstraintWeights_1:float=0.0924
    ConstraintWeights_2:float=0.2519
    ConstraintWeights_3:float=0.4288
    ConstraintWeights_4:float=0.1565
    ConstraintWeights_5:float=0.0008
    ConstraintWeights_6:float=0.0301
    ConstraintWeights_7:float=0.0028
   ) -> (Pose) {
    alphaBoolBlend {blendOption=Linear}
    boneToModify {boneName=master_C_chest_pouch_jnt}
    constraintSetup [
     {targetBone {boneName=pelvis}}
     {targetBone {boneName=spine_01}}
     {targetBone {boneName=spine_02}}
     {targetBone {boneName=spine_03}}
     {targetBone {boneName=spine_04}}
     {targetBone {boneName=spine_05}}
     {targetBone {boneName=clavicle_r}}
     {targetBone {boneName=deform_pec_r}}
    ]
    constraintWeights [1,1,1,1,1,1,1,1]
   }
   n13 = "Constraint - master_L_chest_pouch_jnt (PPPPPP)"!(
    ComponentPose=@n12
    Alpha=1
    ConstraintWeights_0=0.0416
    ConstraintWeights_1:float=0.0946
    ConstraintWeights_2:float=0.2628
    ConstraintWeights_3:float=0.4386
    ConstraintWeights_4:float=0.1609
    ConstraintWeights_5:float=0.0015
   ) -> (Pose) {
    alphaBoolBlend {blendOption=Linear}
    boneToModify {boneName=master_L_chest_pouch_jnt}
    constraintSetup [
     {targetBone {boneName=pelvis}}
     {targetBone {boneName=spine_01}}
     {targetBone {boneName=spine_02}}
     {targetBone {boneName=spine_03}}
     {targetBone {boneName=spine_04}}
     {targetBone {boneName=clavicle_l}}
    ]
    constraintWeights [1,1,1,1,1,1]
   }
   n14 = "Constraint - master_gun_holster_jnt (PPPPP)"!(
    ComponentPose=@n13
    Alpha=1
    ConstraintWeights_0=0.3028
    ConstraintWeights_1:float=0.0008
    ConstraintWeights_2:float=0.5495
    ConstraintWeights_3:float=0.035
    ConstraintWeights_4:float=0.1118
   ) -> (Pose) {
    alphaBoolBlend {blendOption=Linear}
    boneToModify {boneName=master_gun_holster_jnt}
    constraintSetup [
     {targetBone {boneName=pelvis}}
     {targetBone {boneName=thigh_r}}
     {targetBone {boneName=thigh_twist_01_r}}
     {targetBone {boneName=deform_glute_r}}
     {targetBone {boneName=deform_groin_r}}
    ]
    constraintWeights [1,1,1,1,1]
   }
   n15 = "Constraint - master_C_chest_loop_jnt (PPPPPPPP)"!(
    ComponentPose=@n14
    Alpha=1
    ConstraintWeights_0=0.053
    ConstraintWeights_1:float=0.1139
    ConstraintWeights_2:float=0.2425
    ConstraintWeights_3:float=0.327
    ConstraintWeights_4:float=0.1263
    ConstraintWeights_5:float=0.0681
    ConstraintWeights_6:float=0.0086
    ConstraintWeights_7:float=0.0608
   ) -> (Pose) {
    alphaBoolBlend {blendOption=Linear}
    boneToModify {boneName=master_C_chest_loop_jnt}
    constraintSetup [
     {targetBone {boneName=pelvis}}
     {targetBone {boneName=spine_01}}
     {targetBone {boneName=spine_02}}
     {targetBone {boneName=spine_03}}
     {targetBone {boneName=spine_04}}
     {targetBone {boneName=clavicle_l}}
     {targetBone {boneName=upperarm_twist_01_l}}
     {targetBone {boneName=deform_pec_l}}
    ]
    constraintWeights [1,1,1,1,1,1,1,1]
   }
   n16 = "Constraint - master_L_chest_loop_jnt (PPPPPPPP)"!(
    ComponentPose=@n15
    Alpha=1
    ConstraintWeights_0=0.0598
    ConstraintWeights_1:float=0.119
    ConstraintWeights_2:float=0.247
    ConstraintWeights_3:float=0.3112
    ConstraintWeights_4:float=0.1194
    ConstraintWeights_5:float=0.0727
    ConstraintWeights_6:float=0.0207
    ConstraintWeights_7:float=0.0502
   ) -> (Pose) {
    alphaBoolBlend {blendOption=Linear}
    boneToModify {boneName=master_L_chest_loop_jnt}
    constraintSetup [
     {targetBone {boneName=pelvis}}
     {targetBone {boneName=spine_01}}
     {targetBone {boneName=spine_02}}
     {targetBone {boneName=spine_03}}
     {targetBone {boneName=spine_04}}
     {targetBone {boneName=clavicle_l}}
     {targetBone {boneName=upperarm_twist_01_l}}
     {targetBone {boneName=deform_pec_l}}
    ]
    constraintWeights [1,1,1,1,1,1,1,1]
   }
   n17 = "Constraint - master_R_chest_loop_jnt (PPPPPPPP)"!(
    ComponentPose=@n16
    Alpha=1
    ConstraintWeights_0=0.0514
    ConstraintWeights_1:float=0.1105
    ConstraintWeights_2:float=0.2473
    ConstraintWeights_3:float=0.348
    ConstraintWeights_4:float=0.1317
    ConstraintWeights_5:float=0.0572
    ConstraintWeights_6:float=0.0075
    ConstraintWeights_7:float=0.0464
   ) -> (Pose) {
    alphaBoolBlend {blendOption=Linear}
    boneToModify {boneName=master_R_chest_loop_jnt}
    constraintSetup [
     {targetBone {boneName=pelvis}}
     {targetBone {boneName=spine_01}}
     {targetBone {boneName=spine_02}}
     {targetBone {boneName=spine_03}}
     {targetBone {boneName=spine_04}}
     {targetBone {boneName=clavicle_l}}
     {targetBone {boneName=upperarm_twist_01_l}}
     {targetBone {boneName=deform_pec_l}}
    ]
    constraintWeights [1,1,1,1,1,1,1,1]
   }
   Save cached pose 'StartPose'(Pose=Component To Local(ComponentPose=@n17)) {cacheName=StartPose}
  ]}
 EventGraph {nodes [Event BlueprintUpdateAnimation()]}
]
variables []
