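# Physics asset for the male medium AbstractMirror_Era face accessory, simulating a headband's left and right ribbon tails.
# Each side is a five-box chain with limited swing and twist, rooted on small master joint capsules under the head.
# Capsules on the head, neck, upper spine, clavicles and upper arms form the body shapes around the ribbons.
# Constraints between those body parts and master joints leave all angular motion free.
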PhysicsAsset /BRCosmetics/Characters/Player/Male/Medium/Bodies/M_MED_AbstractMirror_Era/Meshes/Parts/M_MED_AbstractMirror_Era_FaceAcc_Physics
bodies {
 clavicle_l [{shapeType=Capsule, length=29, radius=7, shapeName=clavicle_l_capsule, center=<14.9535 1.0882 9.8013>, rotation=<rot 80 -179.9984 -179.9985>}]
 clavicle_r [{
   shapeType=Capsule
   length=29
   radius=7
   shapeName=clavicle_r_capsule
   center=<-17.5198 -1.4378 -9.4903>
   rotation=<rot 80 -179.9984 -179.9988>
  }]
 dyn_l_headband_1_jnt [{shapeType=Box, extentX=4, extentY=1, extentZ=3, shapeName=dyn_l_headband_1_jnt_box, center=<2.4575 0.5 0>, rotation=<rot 0 10.0001 0>}]
 dyn_l_headband_2_jnt [{shapeType=Box, extentX=4, extentY=1.25, extentZ=3, shapeName=dyn_l_headband_2_jnt_box, center=<3.1177 0 0>}]
 dyn_l_headband_3_jnt [{shapeType=Box, extentX=5, extentY=2, extentZ=3, shapeName=dyn_l_headband_3_jnt_box, center=<3 0 0>}]
 dyn_l_headband_4_jnt [{shapeType=Box, extentX=6, extentY=2, extentZ=3, shapeName=dyn_l_headband_4_jnt_box, center=<3.5 0 0>}]
 dyn_l_headband_5_jnt [{shapeType=Box, extentX=5, extentY=1, extentZ=3, shapeName=dyn_l_headband_5_jnt_box, center=<3 0 0>}]
 dyn_r_headband_1_jnt [{shapeType=Box, extentX=4, extentY=1, extentZ=3, shapeName=dyn_r_headband_1_jnt_box, center=<-2.5242 -0.5 0>, rotation=<rot 0 10.0001 0>}]
 dyn_r_headband_2_jnt [{shapeType=Box, extentX=4, extentY=1.25, extentZ=3, shapeName=dyn_r_headband_2_jnt_box, center=<-3.17 0 0>}]
 dyn_r_headband_3_jnt [{shapeType=Box, extentX=5, extentY=2, extentZ=3, shapeName=dyn_r_headband_3_jnt_box, center=<-3 0 0>}]
 dyn_r_headband_4_jnt [{shapeType=Box, extentX=6, extentY=2, extentZ=3, shapeName=dyn_r_headband_4_jnt_box, center=<-3.5 0 0>}]
 dyn_r_headband_5_jnt [{shapeType=Box, extentX=5, extentY=1, extentZ=3, shapeName=dyn_r_headband_5_jnt_box, center=<-3 0 0>}]
 head [{shapeType=Capsule, length=11.5, radius=7, shapeName=head_capsule, center=<1.9314 -1.4634 0>, rotation=<rot -90 180 -150.0003>}]
 master_headband_ribbons_jnt [{shapeType=Capsule, length=0.505, radius=0.505, shapeName=master_headband_ribbons_jnt_capsule, rotation=<rot 0 -90 0>}]
 master_l_headband_jnt [{shapeType=Capsule, length=0.505, radius=0.505, shapeName=master_l_headband_jnt_capsule, rotation=<rot 0 -90 0>}]
 master_r_headband_jnt [{shapeType=Capsule, length=0.505, radius=0.505, shapeName=master_r_headband_jnt_capsule, rotation=<rot 0 -90 0>}]
 neck_01 [{shapeType=Capsule, length=6, radius=4, shapeName=neck_01_capsule, center=<0 -0.8449 0>, rotation=<rot -90 179.9851 -179.9852>}]
 neck_02 [{shapeType=Capsule, length=6, radius=4, shapeName=neck_02_capsule, rotation=<rot -90 153.9103 -153.9105>}]
 spine_04 [{shapeType=Capsule, length=10.5, radius=10, shapeName=spine_04_capsule, center=<4.4625 -2.8648 0>}]
 spine_05 [{shapeType=Capsule, length=29.5, radius=10, shapeName=spine_05_capsule, center=<2.6994 -2.0247 0>}]
 upperarm_l [{shapeType=Capsule, length=10, radius=7, shapeName=upperarm_l_capsule, center=<7.2174 0 -1.2726>, rotation=<rot -80 179.9997 -179.9997>}]
 upperarm_r [{shapeType=Capsule, length=10, radius=7, shapeName=upperarm_r_capsule, center=<-4.931 0 0.8695>, rotation=<rot -80 179.9997 -179.9997>}]
}
constraints [
 {bone1Name=dyn_l_headband_1_jnt, bone2Name=master_l_headband_jnt, swing1LimitDegrees=25, swing2LimitDegrees=10, twistLimitDegrees=25}
 {bone1Name=dyn_l_headband_2_jnt, bone2Name=dyn_l_headband_1_jnt, swing1LimitDegrees=25, swing2LimitDegrees=10, twistLimitDegrees=25}
 {bone1Name=dyn_l_headband_3_jnt, bone2Name=dyn_l_headband_2_jnt, swing1LimitDegrees=25, swing2LimitDegrees=15, twistLimitDegrees=25}
 {bone1Name=dyn_l_headband_4_jnt, bone2Name=dyn_l_headband_3_jnt, swing1LimitDegrees=25, swing2LimitDegrees=15, twistLimitDegrees=25}
 {bone1Name=dyn_l_headband_5_jnt, bone2Name=dyn_l_headband_4_jnt, swing1LimitDegrees=25, swing2LimitDegrees=15, twistLimitDegrees=25}
 {bone1Name=dyn_r_headband_1_jnt, bone2Name=master_r_headband_jnt, swing1LimitDegrees=25, swing2LimitDegrees=10, twistLimitDegrees=25}
 {bone1Name=dyn_r_headband_2_jnt, bone2Name=dyn_r_headband_1_jnt, swing1LimitDegrees=25, swing2LimitDegrees=10, twistLimitDegrees=25}
 {bone1Name=dyn_r_headband_3_jnt, bone2Name=dyn_r_headband_2_jnt, swing1LimitDegrees=25, swing2LimitDegrees=15, twistLimitDegrees=25}
 {bone1Name=dyn_r_headband_4_jnt, bone2Name=dyn_r_headband_3_jnt, swing1LimitDegrees=25, swing2LimitDegrees=15, twistLimitDegrees=25}
 {bone1Name=dyn_r_headband_5_jnt, bone2Name=dyn_r_headband_4_jnt, swing1LimitDegrees=25, swing2LimitDegrees=15, twistLimitDegrees=25}
 {
  bone1Name=master_headband_ribbons_jnt
  bone2Name=head
  swing1LimitDegrees=45
  swing1Motion=Free
  swing2LimitDegrees=45
  swing2Motion=Free
  twistLimitDegrees=45
  twistMotion=Free
 }
 {
  bone1Name=master_l_headband_jnt
  bone2Name=master_headband_ribbons_jnt
  swing1LimitDegrees=45
  swing1Motion=Free
  swing2LimitDegrees=45
  swing2Motion=Free
  twistLimitDegrees=45
  twistMotion=Free
 }
 {
  bone1Name=master_r_headband_jnt
  bone2Name=master_headband_ribbons_jnt
  swing1LimitDegrees=45
  swing1Motion=Free
  swing2LimitDegrees=45
  swing2Motion=Free
  twistLimitDegrees=45
  twistMotion=Free
 }
 {
  bone1Name=spine_05
  bone2Name=spine_04
  swing1LimitDegrees=45
  swing1Motion=Free
  swing2LimitDegrees=45
  swing2Motion=Free
  twistLimitDegrees=45
  twistMotion=Free
 }
 {
  bone1Name=neck_01
  bone2Name=spine_05
  swing1LimitDegrees=45
  swing1Motion=Free
  swing2LimitDegrees=45
  swing2Motion=Free
  twistLimitDegrees=45
  twistMotion=Free
 }
 {
  bone1Name=clavicle_l
  bone2Name=spine_05
  swing1LimitDegrees=45
  swing1Motion=Free
  swing2LimitDegrees=45
  swing2Motion=Free
  twistLimitDegrees=45
  twistMotion=Free
 }
 {
  bone1Name=clavicle_r
  bone2Name=spine_05
  swing1LimitDegrees=45
  swing1Motion=Free
  swing2LimitDegrees=45
  swing2Motion=Free
  twistLimitDegrees=45
  twistMotion=Free
 }
 {
  bone1Name=upperarm_l
  bone2Name=clavicle_l
  swing1LimitDegrees=45
  swing1Motion=Free
  swing2LimitDegrees=45
  swing2Motion=Free
  twistLimitDegrees=45
  twistMotion=Free
 }
 {
  bone1Name=upperarm_r
  bone2Name=clavicle_r
  swing1LimitDegrees=45
  swing1Motion=Free
  swing2LimitDegrees=45
  swing2Motion=Free
  twistLimitDegrees=45
  twistMotion=Free
 }
 {
  bone1Name=neck_02
  bone2Name=neck_01
  swing1LimitDegrees=45
  swing1Motion=Free
  swing2LimitDegrees=45
  swing2Motion=Free
  twistLimitDegrees=45
  twistMotion=Free
 }
 {
  bone1Name=head
  bone2Name=neck_02
  swing1LimitDegrees=45
  swing1Motion=Free
  swing2LimitDegrees=45
  swing2Motion=Free
  twistLimitDegrees=45
  twistMotion=Free
 }
]
